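# Physics asset for the face accessory mesh of the female medium Durrburger Worker player cosmetic.
# It holds a single body on the head bone, shaped as one capsule offset from the bone origin and rotated a quarter turn.
# It defines no constraints, so it provides a simple head collision volume rather than a jointed ragdoll.
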PhysicsAsset /BRCosmetics/Characters/Player/Female/Medium/Bodies/F_MED_DurrburgerWorker/Meshes/Parts/F_MED_Durrburger_Worker_FaceAcc_PhysicsAsset
bodies {head [{shapeType=Capsule, length=6.9312, radius=6.3118, center=<0 3.1884 0>, rotation=<rot 90 0 0>}]}
constraints []
